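# Animation Blueprint for the ChiveFlake gadget prop, an emote vehicle with front wheels that turn with its owning actor.
# A linear bool blend on bIsMoving2D picks between the InMotion and default montage slots, then adds yaw to both front wheel bones at half alpha.
# Each update, Adjust Emote Vehicle Wheel Rotation derives that yaw and stores actor rotation.
# Current Wheel Rotation feeds that call but is never set here.
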AnimBlueprint /BRCosmetics/Animation/Props/Gadgets_ChiveFlake/Meshes/ChiveFlake_Gadget_Skeleton_AnimBP
target_skeleton=&/ChiveFlake_Gadget_Skeleton
graphs [
 AnimGraph {nodes [
   n1 = Slot 'InMotion'() {slotName=InMotion}
   n2 = Slot 'DefaultSlot'()
   n3 = Get bIsMoving2D()
   n4 = Get Wheel Rotation() -> (Wheel Rotation:double)
   n5 = "Transform (Modify) Bone - Bone: JNT_FrontWheelRight"!(
    ComponentPose=Local To Component(LocalPose=Blend Poses by bool(BlendPose_0=@n1, BlendPose_1=@n2, bActiveValue=@n3) {blendType=Linear})
    Alpha=0.5
    Translation="0,0,0"
    Rotation_Roll:float=0
    Rotation_Pitch:float=0
    Rotation_Yaw:float=@n4
    Scale="1,1,1"
   ) -> (Pose) {rotationMode=BMM_Additive, alphaBoolBlend {blendOption=Linear}, boneToModify {boneName=JNT_FrontWheelRight}}
   Output Pose(Result=Component To Local(ComponentPose="Transform (Modify) Bone - Bone: JNT_FrontWheelLeft"!(ComponentPose=@n5, Alpha=0.5, Translation="0,0,0", Rotation_Roll:float=0, Rotation_Pitch:float=0, Rotation_Yaw:float=@n4, Scale="1,1,1") -> (Pose) {rotationMode=BMM_Additive, alphaBoolBlend {blendOption=Linear}, boneToModify {boneName=JNT_FrontWheelLeft}}))
  ]}
 EventGraph {nodes [
   n1 = Event BlueprintUpdateAnimation()
   > n2 = Adjust Emote Vehicle Wheel Rotation!(
    execute
    Owning Actor:Actor Object Reference=Get Owning Actor() -> (Owning Actor:Actor Object Reference)
    CurrentWheelRotation:double=Get Current Wheel Rotation() -> (Current Wheel Rotation:double)
    DeltaTime:double=@n1.DeltaTimeX
    CurrentActorRotation:double=Get Current Actor Rotation() -> (Current Actor Rotation:double)
   ) -> (then, NewActorRotation:double, NewWheelRotation:double)
   > Set Current Actor Rotation(Current Actor Rotation:double=@n2) -> (Output_Get:double)
   > Set Wheel Rotation(Wheel Rotation:double=@n2.NewWheelRotation) -> (Output_Get:double)
   Event BlueprintInitializeAnimation()
   > Set Owning Actor(Owning Actor:Actor Object Reference=GetOwningActor()) -> (Output_Get:Actor Object Reference)
   TryGetPawnOwner()
  ]}
]
variables [
 Current Actor Rotation {type=number}
 Current Wheel Rotation {type=number}
 Owning Actor {type=Actor, default=null}
 Wheel Rotation {type=number}
]
